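# Animation Blueprint for the female DarkNinja head, correcting hair bones as the character moves.
# The output applies additive Transform Bone corrections to dyn_hair_07/09 on the input pose, weighted by head, spine-while-targeting and arm-raise alphas.
# Updates derive those alphas from bone deltas against reference transforms, zeroing arm alphas during emotes.
# An AnimDynamics chain and hood-equipped blend exist but never reach Output Pose; the right-spine alpha is also unused.
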AnimBlueprint /BRCosmetics/Characters/Player/Female/Medium/Heads/F_MED_ASN_DarkNinja_Head_01/Meshes/F_MED_ASN_DarkNinja_Head_01_Dyn_AnimBP
target_skeleton=/Game/Characters/Player/Common/Fortnite_Base_Head/Fortnite_Base_Head_Export_Skeleton
graphs [
 AnimGraph {nodes [
   n1 = Anim Dynamics - Bone: dyn_hair_01(ComponentPose=Local To Component(LocalPose=Copy Pose From Mesh() {bUseAttachedParent=true, bUseMeshPose=true}), Alpha=0.25) {
    bGravityOverrideInSimSpace=true
    alphaBoolBlend {blendOption=Linear}
    boundBone {boneName=dyn_hair_01}
    physicsBodyDefinitions [{
      sphereCollisionRadius=0
      boundBone {boneName=dyn_hair_01}
      boxExtents=<20 20 20>
      constraintSetup {twistAxis=AxisZ, angularLimitsMax=<10 10 0.001>, angularLimitsMin=<-0.001 -10 -0.001>, angularTarget=<0 0 0>}
      localJointOffset=<-0.001 -0.001 -5>
     }]
   }
   n2 = Anim Dynamics - Bone: dyn_hair_02(ComponentPose=@n1, Alpha=0.5) {
    bGravityOverrideInSimSpace=true
    alphaBoolBlend {blendOption=Linear}
    boundBone {boneName=dyn_hair_02}
    physicsBodyDefinitions [{
      sphereCollisionRadius=0
      boundBone {boneName=dyn_hair_02}
      boxExtents=<20 20 20>
      constraintSetup {twistAxis=AxisZ, angularLimitsMax=<10 20 0.001>, angularLimitsMin=<-0.001 -20 -0.001>, angularTarget=<0 0 0>}
      localJointOffset=<-0.001 -0.001 -5>
     }]
   }
   n3 = Anim Dynamics - Bone: dyn_hair_03(ComponentPose=@n2, Alpha=0.5, ExternalForce="0,0,0") {
    bGravityOverrideInSimSpace=true
    alphaBoolBlend {blendOption=Linear}
    boundBone {boneName=dyn_hair_03}
    physicsBodyDefinitions [{
      sphereCollisionRadius=0
      boundBone {boneName=dyn_hair_03}
      boxExtents=<20 20 20>
      constraintSetup {twistAxis=AxisZ, angularLimitsMax=<12 20 0.001>, angularLimitsMin=<-0.001 -20 -0.001>, angularTarget=<0 0 0>}
      localJointOffset=<-0.001 -0.001 -5>
     }]
   }
   n4 = Anim Dynamics - Bone: dyn_hair_04(ComponentPose=@n3, Alpha=0.5, ExternalForce="0,0,0") {
    bGravityOverrideInSimSpace=true
    alphaBoolBlend {blendOption=Linear}
    boundBone {boneName=dyn_hair_04}
    physicsBodyDefinitions [{
      sphereCollisionRadius=0
      boundBone {boneName=dyn_hair_04}
      boxExtents=<20 20 20>
      constraintSetup {twistAxis=AxisZ, angularLimitsMax=<15 20 0.001>, angularLimitsMin=<-0.001 -20 -0.001>, angularTarget=<0 0 0>}
      localJointOffset=<-0.001 -0.001 -5>
     }]
   }
   n5 = Anim Dynamics - Bone: dyn_hair_05(ComponentPose=@n4, Alpha=0.3, ExternalForce="0,0,0") {
    bGravityOverrideInSimSpace=true
    alphaBoolBlend {blendOption=Linear}
    boundBone {boneName=dyn_hair_05}
    physicsBodyDefinitions [{
      sphereCollisionRadius=0
      boundBone {boneName=dyn_hair_05}
      boxExtents=<20 20 20>
      constraintSetup {twistAxis=AxisZ, angularLimitsMax=<10 10 0.001>, angularLimitsMin=<-0.001 -10 -0.001>, angularTarget=<0 0 0>}
      localJointOffset=<-0.001 -0.001 -5>
     }]
   }
   n6 = Anim Dynamics - Bone: dyn_hair_06(ComponentPose=@n5, Alpha=0.3, ExternalForce="0,0,0") {
    bGravityOverrideInSimSpace=true
    alphaBoolBlend {blendOption=Linear}
    boundBone {boneName=dyn_hair_06}
    physicsBodyDefinitions [{
      sphereCollisionRadius=0
      boundBone {boneName=dyn_hair_06}
      boxExtents=<20 20 20>
      constraintSetup {twistAxis=AxisZ, angularLimitsMax=<10 10 0.001>, angularLimitsMin=<-0.001 -10 -0.001>, angularTarget=<0 0 0>}
      localJointOffset=<-0.001 -0.001 -5>
     }]
   }
   n7 = Anim Dynamics - Bone: dyn_hair_09(ComponentPose=@n6, Alpha=0.25, ExternalForce="0,0,0") {
    bGravityOverrideInSimSpace=true
    alphaBoolBlend {blendOption=Linear}
    boundBone {boneName=dyn_hair_09}
    physicsBodyDefinitions [{
      sphereCollisionRadius=0
      boundBone {boneName=dyn_hair_09}
      boxExtents=<30 30 30>
      constraintSetup {angularTargetAxis=AxisY, twistAxis=AxisZ, angularLimitsMax=<25 60 0.001>, angularLimitsMin=<-25 0 -0.001>, angularTarget=<0 0 0>}
      localJointOffset=<-0.001 -0.001 -5>
     }]
   }
   n8 = Anim Dynamics - Bone: dyn_hair_07(ComponentPose=@n7, Alpha=0.25, ExternalForce="0,0,0") {
    bGravityOverrideInSimSpace=true
    alphaBoolBlend {blendOption=Linear}
    boundBone {boneName=dyn_hair_07}
    physicsBodyDefinitions [{
      sphereCollisionRadius=0
      boundBone {boneName=dyn_hair_07}
      boxExtents=<30 30 30>
      constraintSetup {angularTargetAxis=AxisY, twistAxis=AxisZ, angularLimitsMax=<25 0 0.001>, angularLimitsMin=<-25 -60 -0.001>, angularTarget=<0 0 0>}
      localJointOffset=<-0.001 -0.001 -5>
     }]
   }
   Blend Poses by bool(
    BlendPose_0=Component To Local()
    BlendPose_1=Component To Local(ComponentPose=@n8)
    bActiveValue=Get Is Hood equipped() -> (Is Hood equipped:bool)
   ) {blendType=Linear}
   AnimGraphNode_SubInput_0 = Input Pose()
   n9 = "Transform (Modify) Bone - Bone: dyn_hair_09"(
    ComponentPose=Local To Component(LocalPose=@AnimGraphNode_SubInput_0)
    Alpha=Get Head turn to the right Alpha() -> (Head turn to the right Alpha:double)
    Translation="0,0,0"
    Rotation="19.000,0.005,-0.030"
    Scale="1,1,1"
   ) {
    rotationMode=BMM_Additive
    rotationSpace=BCS_BoneSpace
    translationMode=BMM_Additive
    translationSpace=BCS_BoneSpace
    alphaBoolBlend {blendOption=Linear}
    boneToModify {boneName=dyn_hair_09}
    rotation=<rot 19 0.005 -0.0301>
   }
   n10 = "Transform (Modify) Bone - Bone: dyn_hair_07"(
    ComponentPose=@n9
    Alpha=Get Head turn to the left Alpha() -> (Head turn to the left Alpha:double)
    Translation="0,0,0"
    Rotation="-19,0,0"
    Scale="1,1,1"
   ) {
    rotationMode=BMM_Additive
    rotationSpace=BCS_BoneSpace
    translationMode=BMM_Additive
    translationSpace=BCS_BoneSpace
    alphaBoolBlend {blendOption=Linear}
    boneToModify {boneName=dyn_hair_07}
    rotation {pitch=-19}
   }
   n11 = Get Head look Up Alpha() -> (Head look Up Alpha:double)
   n12 = "Transform (Modify) Bone - Bone: dyn_hair_07"(ComponentPose=@n10, Alpha=@n11, Translation="-1,-0.12,-0.106", Rotation="-2.483,3.823,14.674", Scale="1,1,1") {
    rotationMode=BMM_Additive
    rotationSpace=BCS_BoneSpace
    translationMode=BMM_Additive
    alphaBoolBlend {blendOption=Linear}
    boneToModify {boneName=dyn_hair_07}
    rotation=<rot -2.483 3.823 14.6739>
    translation=<-1 -0.12 -0.106>
   }
   n13 = "Transform (Modify) Bone - Bone: dyn_hair_09"(ComponentPose=@n12, Alpha=@n11, Translation="0.558,-0.080,0.014", Rotation="-1.422,-0.814,15.027", Scale="1,1,1") {
    rotationMode=BMM_Additive
    rotationSpace=BCS_BoneSpace
    translationMode=BMM_Additive
    alphaBoolBlend {blendOption=Linear}
    boneToModify {boneName=dyn_hair_09}
    rotation=<rot -1.422 -0.814 15.0269>
    translation=<0.558 -0.08 0.014>
   }
   n14 = Get Head look Down Alpha() -> (Head look Down Alpha:double)
   n15 = "Transform (Modify) Bone - Bone: dyn_hair_07"(ComponentPose=@n13, Alpha=@n14, Translation="0,0,0", Rotation="-4.813,1.436,-24.992", Scale="1,1,1") {
    rotationMode=BMM_Additive
    rotationSpace=BCS_BoneSpace
    translationMode=BMM_Additive
    alphaBoolBlend {blendOption=Linear}
    boneToModify {boneName=dyn_hair_07}
    rotation=<rot -4.813 1.436 -24.9921>
   }
   n16 = "Transform (Modify) Bone - Bone: dyn_hair_09"(ComponentPose=@n15, Alpha=@n14, Translation="0,0,0", Rotation="4.647,-1.846,-25.075", Scale="1,1,1") {
    rotationMode=BMM_Additive
    rotationSpace=BCS_BoneSpace
    translationMode=BMM_Additive
    alphaBoolBlend {blendOption=Linear}
    boneToModify {boneName=dyn_hair_09}
    rotation=<rot 4.647 -1.846 -25.0752>
   }
   n17 = "Transform (Modify) Bone - Bone: dyn_hair_09"(
    ComponentPose=@n16
    Alpha=Get Spine rotation to the Left while Targeting Alpha() -> (Spine rotation to the Left while Targeting Alpha:double)
    Translation="0,0,0"
    Rotation="5.000,0.003,-0.000"
    Scale="1,1,1"
   ) {
    rotationMode=BMM_Additive
    rotationSpace=BCS_BoneSpace
    translationMode=BMM_Additive
    alphaBoolBlend {blendOption=Linear}
    boneToModify {boneName=dyn_hair_09}
    rotation=<rot 5.0001 0.003 -0.0002>
   }
   n18 = "Transform (Modify) Bone - Bone: dyn_hair_07"(
    ComponentPose=@n17
    Alpha=Get L Arm Raise Alpha() -> (L Arm Raise Alpha:float)
    Translation="0,0,0"
    Rotation="-5.000,0.002,-0.005"
    Scale="1,1,1"
   ) {
    rotationMode=BMM_Additive
    rotationSpace=BCS_BoneSpace
    translationMode=BMM_Additive
    alphaBoolBlend {blendOption=Linear}
    boneToModify {boneName=dyn_hair_07}
    rotation=<rot -4.9999 0.002 -0.0051>
   }
   n19 = "Transform (Modify) Bone - Bone: dyn_hair_09"(
    ComponentPose=@n18
    Alpha=Get R Arm Raise Alpha() -> (R Arm Raise Alpha:float)
    Translation="0,0,0"
    Rotation="5.000,0.003,-0.000"
    Scale="1,1,1"
   ) {
    rotationMode=BMM_Additive
    rotationSpace=BCS_BoneSpace
    translationMode=BMM_Additive
    alphaBoolBlend {blendOption=Linear}
    boneToModify {boneName=dyn_hair_09}
    rotation=<rot 5.0001 0.003 -0.0003>
   }
   n20 = "Transform (Modify) Bone - Bone: dyn_hair_07"(
    ComponentPose=@n19
    Alpha=Get Head Tilt To Right Alpha() -> (Head Tilt To Right Alpha:double)
    Translation="0.144,0.050,0.028"
    Rotation="-8,0.004,-0.002"
    Scale="1,1,1"
   ) {
    rotationMode=BMM_Additive
    rotationSpace=BCS_BoneSpace
    translationMode=BMM_Additive
    alphaBoolBlend {blendOption=Linear}
    boneToModify {boneName=dyn_hair_07}
    rotation=<rot -8 0.004 -0.002>
    translation=<0.144 0.05 0.028>
   }
   Output Pose(Result=Component To Local(ComponentPose=@n20)) {layerGroup=None}
  ]}
 EventGraph {nodes [
   Event BlueprintUpdateAnimation()
   > n1 = Sequence!(execute) -> (then_0, then_1, then_2:Exec)
   n2 = Get Player Pawn() -> (Player Pawn:Player Pawn Athena Object Reference)
   > Is Valid(InputObject=@n2)
   > Do Once()
   > Set Is Hood equipped(Is Hood equipped:bool="Not Equal (Object)"(A:Object Reference=GetSkeletalMeshForPartType(self=@n2, PartType=Hat), B:Object Reference) -> (ReturnValue:bool=~) {is=title K2Node_PromotableOperator}) -> (Output_Get:bool)
   n3 = Branch(execute=@n1.then_2, Condition=Get bIsPlayingEmote())
   > Set R Arm Raise Alpha(R Arm Raise Alpha:float=0) -> (Output_Get:float)
   > Set L Arm Raise Alpha(L Arm Raise Alpha:float=0) -> (Output_Get:float)
   Set R Arm Raise Alpha(execute=@n3.else, R Arm Raise Alpha=Get R Arm Raise Alpha())
   > Set L Arm Raise Alpha(L Arm Raise Alpha=Get L Arm Raise Alpha())
   n4 = Get Skeletal Mesh() -> (Skeletal Mesh:Skeletal Mesh Component Object Reference)
   Set Head turn to the right Alpha(
    execute=@n1.then_1
    Head turn to the right Alpha:double=GetAlphaFromReferenceTransformDelta(SkeletalMeshComponent=@n4, BoneName=head, BaseName=spine_05, DeltaComponentType=EulerX, InputRange=(X=0,Y=-90), TwistAxis="0,0,1")
   ) -> (Output_Get:double)
   > Set Head turn to the left Alpha(Head turn to the left Alpha:double=GetAlphaFromReferenceTransformDelta(SkeletalMeshComponent=@n4, BoneName=head, BaseName=neck_01, DeltaComponentType=EulerX, InputRange=(X=10,Y=90), TwistAxis="0,0,1")) -> (Output_Get:double)
   > Set Head look Up Alpha(Head look Up Alpha:double=GetAlphaFromReferenceTransformDelta(SkeletalMeshComponent=@n4, BoneName=head, BaseName=spine_05, InputRange=(X=0,Y=-45), TwistAxis="0,0,1")) -> (Output_Get:double)
   > Set Head look Down Alpha(Head look Down Alpha:double=GetAlphaFromReferenceTransformDelta(SkeletalMeshComponent=@n4, BoneName=head, BaseName=spine_05, InputRange=(X=0,Y=20), TwistAxis="0,0,1")) -> (Output_Get:double)
   > Set Spine rotation to the Left while Targeting Alpha(Spine rotation to the Left while Targeting Alpha:double=float * float(
     A=GetAlphaFromReferenceTransformDelta(SkeletalMeshComponent=@n4, BoneName=spine_05, BaseName=pelvis, DeltaComponentType=EulerX, InputRange=(X=-10,Y=7), TwistAxis="0,0,1")
     B:bool=Get bIsTargeting()
    ) {is=title K2Node_PromotableOperator}) -> (Output_Get:double)
   > Set Spine rotation to the Right while Targeting Alpha(Spine rotation to the Right while Targeting Alpha:double=float * float(
     A=GetAlphaFromReferenceTransformDelta(SkeletalMeshComponent=@n4, BoneName=spine_05, BaseName=pelvis, DeltaComponentType=EulerX, InputRange=(X=-10,Y=-45), TwistAxis="0,0,1")
     B=Get bIsTargeting()
    )) -> (Output_Get:double)
   > Set R Arm Raise Alpha(R Arm Raise Alpha=GetAlphaFromReferenceTransformDelta(SkeletalMeshComponent=@n4, BoneName=upperarm_r, BaseName=spine_05, InputRange=(X=60,Y=105), TwistAxis="0,0,1"))
   > Set L Arm Raise Alpha(L Arm Raise Alpha=GetAlphaFromReferenceTransformDelta(SkeletalMeshComponent=@n4, BoneName=upperarm_l, BaseName=spine_05, InputRange=(X=60,Y=105), TwistAxis="0,0,1"))
   > Set Head Tilt To Right Alpha(Head Tilt To Right Alpha:double=GetAlphaFromReferenceTransformDelta(SkeletalMeshComponent=@n4, BoneName=head, BaseName=spine_05, DeltaComponentType=EulerY, InputRange=(X=0,Y=30), TwistAxis="0,0,1")) -> (Output_Get:double)
   Event BlueprintInitializeAnimation()
   > n5 = Sequence!(execute) -> (then_0, then_1)
   n6 = TryGetPawnOwner()
   > Is Valid(InputObject=@n6)
   > n7 = Cast To PlayerPawn_Athena(Object=@n6)
   > Set Player Pawn(Player Pawn:Player Pawn Athena Object Reference=@n7) -> (Output_Get:Player Pawn Athena Object Reference)
   Set Skeletal Mesh(execute=@n5.then_1, Skeletal Mesh:Skeletal Mesh Component Object Reference=GetOwnerSkeletalMeshComponent()) -> (Output_Get:Skeletal Mesh Component Object Reference)
  ]}
]
variables [
 Head look Down Alpha {type=number}
 Head look Up Alpha {type=number}
 Head Tilt To Right Alpha {type=number}
 Head turn to the left Alpha {type=number}
 Head turn to the right Alpha {type=number}
 Is Hood equipped {type=boolean}
 L Arm Raise Alpha {type=number}
 Player Pawn {type=PlayerPawn_Athena_C}
 R Arm Raise Alpha {type=number}
 Skeletal Mesh {type=SkeletalMeshComponent}
 Spine rotation to the Left while Targeting Alpha {type=number}
 Spine rotation to the Right while Targeting Alpha {type=number}
]
